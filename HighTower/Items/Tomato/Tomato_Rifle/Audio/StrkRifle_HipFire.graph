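# Sound cue for the hip-fire gunshot of the Tomato StrkRifle, layering close, medium and far recordings.
# Each layer picks one of three random waves, sets its volume through a modulator, then feeds an attenuation node and sound class override into one mixer.
# The close layer alone varies pitch and volume and names an attenuation asset, while medium and far share the Distant fire class.
# Concurrency allows three instances per owner, stopping the oldest.
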SoundCue /HighTower/Items/Tomato/Tomato_Rifle/Audio/StrkRifle_HipFire
root=Mixer_0
nodes [
 {id=Attenuation_2, children [Modulator_2], properties {attenuationSettings=&Fort_Weapon_Projectile_Close_Athena_Att}}
 {id=Attenuation_3, children [Modulator_3]}
 {id=Attenuation_4, children [Modulator_4]}
 {id=Mixer_0, children [SoundClass_2,SoundClass_4,SoundClass_3], properties {inputVolume [1,0.9,0.7]}}
 {id=Modulator_2, children [Random_2], properties {pitchMax=1.1, pitchMin=1, volumeMax=1.2, volumeMin=1.1}}
 {id=Modulator_3, children [Random_0], properties {volumeMax=0.7, volumeMin=0.7}}
 {id=Modulator_4, children [Random_1], properties {volumeMax=0.3, volumeMin=0.3}}
 {id=Random_0, children [WavePlayer_6,WavePlayer_7,WavePlayer_9], properties {weights [1,1,1]}}
 {id=Random_1, children [WavePlayer_0,WavePlayer_1,WavePlayer_2], properties {weights [1,1,1]}}
 {id=Random_2, children [WavePlayer_3,WavePlayer_4,WavePlayer_5], properties {weights [1,1,1]}}
 {
  id=SoundClass_2
  children [Attenuation_2]
  properties {soundClassOverride=&Fort_WeaponRanged_Fire_3P_PSM}
 }
 {
  id=SoundClass_3
  children [Attenuation_4]
  properties {soundClassOverride=&Fort_WeaponRanged_Fire_Distant}
 }
 {
  id=SoundClass_4
  children [Attenuation_3]
  properties {soundClassOverride=&Fort_WeaponRanged_Fire_Distant}
 }
 {id=WavePlayer_0, properties {soundWaveAssetPtr=&/StrkRifle_Fire_Hip_Far_01}}
 {id=WavePlayer_1, properties {soundWaveAssetPtr=&/StrkRifle_Fire_Hip_Far_02}}
 {id=WavePlayer_2, properties {soundWaveAssetPtr=&/StrkRifle_Fire_Hip_Far_03}}
 {id=WavePlayer_3, properties {soundWaveAssetPtr=&/StrkRifle_Fire_Hip_Close_01}}
 {id=WavePlayer_4, properties {soundWaveAssetPtr=&/StrkRifle_Fire_Hip_Close_02}}
 {id=WavePlayer_5, properties {soundWaveAssetPtr=&/StrkRifle_Fire_Hip_Close_03}}
 {id=WavePlayer_6, properties {soundWaveAssetPtr=&/StrkRifle_Fire_Hip_Med_01}}
 {id=WavePlayer_7, properties {soundWaveAssetPtr=&/StrkRifle_Fire_Hip_Med_03}}
 {id=WavePlayer_9, properties {soundWaveAssetPtr=&/StrkRifle_Fire_Hip_Med_02}}
]
properties {
 bDebug=true
 bOverrideConcurrency=true
 maxDistance=2097152
 soundClassObject=&Fort_WeaponRanged_Fire_3P_PSM
 volumeMultiplier=0.7
 concurrencyOverrides {bLimitToOwner=true, maxCount=3, resolutionRule=StopOldest}
}
