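# Sound cue for the Tomato Rifle's low-ammo hip-fire sound, routed to a degraded ranged-weapon low-ammo sound class.
# A mixer layers two low-ammo fire waves, one boosted, with a randomly picked impact sizzle, where Sizzle_02 fills two of three equally weighted slots.
# The root modulator randomizes pitch upward and applies a fixed volume boost.
# A concurrency override blocks new instances once the limit is reached.
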SoundCue /HighTower/Items/Tomato/Tomato_Rifle/Audio/StrkRifle_LowAmmo_HipFire
root=Modulator_0
nodes [
 {id=Mixer_0, children [Random_0,WavePlayer_0,WavePlayer_4], properties {inputVolume [1,1,1.4]}}
 {id=Modulator_0, children [Mixer_0], properties {pitchMax=1.3, pitchMin=1, volumeMax=1.2, volumeMin=1.2}}
 {id=Random_0, children [WavePlayer_2,WavePlayer_2,WavePlayer_3], properties {weights [1,1,1]}}
 {id=WavePlayer_0, properties {soundWaveAssetPtr=&/StrkRifle_Fire_LowAmmo}}
 {id=WavePlayer_2, properties {soundWaveAssetPtr=&/StrkRifle_Impact_Sizzle_02}}
 {id=WavePlayer_3, properties {soundWaveAssetPtr=&/StrkRifle_Impact_Sizzle_03}}
 {id=WavePlayer_4, properties {soundWaveAssetPtr=&/StrkRifle_Fire_LowAmmo_v2}}
]
properties {
 bOverrideConcurrency=true
 maxDistance=2097152
 soundClassObject=/Game/Sounds/FortSoundClassesAndModes/Fort_WeaponRanged_LowAmmo_Degraded_PSM
 volumeMultiplier=1
 concurrencyOverrides {resolutionRule=PreventNew}
}
