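# Sound cue for the Tomato Rifle, named as the reload-end and swap foley sound of its StrkRifle audio set.
# A single wave player feeds the root with one sound wave, with no randomization or modulation.
# It uses the shared Fortnite weapon reload attenuation and weapon foley sound class, with reduced volume, overridden concurrency and the debug flag on.
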SoundCue /HighTower/Items/Tomato/Tomato_Rifle/Audio/StrkRifle_Reload_EndnSwap
root=WavePlayer_1
nodes [{id=WavePlayer_1, properties {soundWaveAssetPtr=&/StrkRifle_Reload_EndnSwap_01}}]
properties {
 attenuationSettings=&Fort_Weapon_Reload_Att
 bDebug=true
 bOverrideConcurrency=true
 maxDistance=2250
 soundClassObject=&Fort_WeaponFoley
 volumeMultiplier=0.7
}
