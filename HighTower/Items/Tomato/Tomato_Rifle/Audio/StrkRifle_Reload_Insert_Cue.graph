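# StrkRifle_Reload_Insert_Cue is the sound cue for the magazine-insert step of the Tomato rifle's reload.
# A single wave player feeds the root directly with one mag-in sample, with no randomization or layering.
# The cue boosts volume, slightly lowers pitch, uses the weapon-foley sound class and shared reload attenuation, and overrides concurrency.
# Its debug flag is switched on.
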SoundCue /HighTower/Items/Tomato/Tomato_Rifle/Audio/StrkRifle_Reload_Insert_Cue
root=WavePlayer_0
nodes [{id=WavePlayer_0, properties {soundWaveAssetPtr=&/StrkRifle_Reload_MagIn_01}}]
properties {
 attenuationSettings=&Fort_Weapon_Reload_Att
 bDebug=true
 bOverrideConcurrency=true
 maxDistance=2250
 pitchMultiplier=0.9
 soundClassObject=&Fort_WeaponFoley
 volumeMultiplier=1.7
}
