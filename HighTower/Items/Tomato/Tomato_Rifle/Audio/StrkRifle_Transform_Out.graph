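# Sound cue for the Tomato rifle's transform-out foley, reusing the StrkRifle ADS-out wave.
# A single wave player feeds a modulator that applies a fixed volume boost with no pitch variation, offset by a lowered cue volume and pitch.
# It uses the weapon reload attenuation and weapon foley sound class, with overridden concurrency that stops the oldest instance beyond two.
# The debug flag is left enabled.
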SoundCue /HighTower/Items/Tomato/Tomato_Rifle/Audio/StrkRifle_Transform_Out
root=Modulator_0
nodes [
 {id=Modulator_0, children [WavePlayer_2], properties {pitchMax=1, pitchMin=1, volumeMax=1.8, volumeMin=1.8}}
 {id=WavePlayer_2, properties {soundWaveAssetPtr=&/StrkRifle_ADSOut_01_M}}
]
properties {
 attenuationSettings=&Fort_Weapon_Reload_Att
 bDebug=true
 bOverrideConcurrency=true
 maxDistance=2250
 pitchMultiplier=0.9
 soundClassObject=&Fort_WeaponFoley
 volumeMultiplier=0.45
 concurrencyOverrides {maxCount=2, resolutionRule=StopOldest}
}
