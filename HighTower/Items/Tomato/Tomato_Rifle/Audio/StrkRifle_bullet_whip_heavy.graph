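# Sound cue for the heavy bullet whip-by of the Tomato rifle (StrkRifle), using close whip-by recordings.
# A quality-level root sends its first two slots to an equal-weight random pick of four waves, while the last two play fixed takes 02 and 01.
# It overrides attenuation to disable distance volume falloff, routes to the ranged projectile effects sound class, and uses the Athena bullet whiz-by concurrency rule.
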SoundCue /HighTower/Items/Tomato/Tomato_Rifle/Audio/StrkRifle_bullet_whip_heavy
root=QualityLevel_0
nodes [
 {id=QualityLevel_0, children [Random_0,Random_0,WavePlayer_8,WavePlayer_9]}
 {id=Random_0, children [WavePlayer_6,WavePlayer_7,WavePlayer_8,WavePlayer_9], properties {weights [1,1,1,1]}}
 {id=WavePlayer_6, properties {soundWaveAssetPtr=&/StrkRifle_WhipBy_Close_04}}
 {id=WavePlayer_7, properties {soundWaveAssetPtr=&/StrkRifle_WhipBy_Close_03}}
 {id=WavePlayer_8, properties {soundWaveAssetPtr=&/StrkRifle_WhipBy_Close_02}}
 {id=WavePlayer_9, properties {soundWaveAssetPtr=&/StrkRifle_WhipBy_Close_01}}
]
properties {
 bOverrideAttenuation=true
 maxDistance=2097152
 soundClassObject=&Fort_WeaponRanged_Projectile_Effects
 volumeMultiplier=1
 attenuationOverrides {bAttenuate=false, stereoSpread=5000}
 concurrencySet [/Game/Sounds/Concurrency_Rules/Athena_BulletWhizBy_Conc]
}
