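; Weapon Blueprint for the Stark assault rifle in Athena, switching sound and effects between hipfire and aim-down-sights (ADS).
; On targeting change it stores the state, prints a debug string and plays the matching transform sound on the weapon mesh.
; The state also selects the muzzle particle color and is written as an integer into the NDC tracer info.
; Local unequip while aiming plays the out-of-ADS sound.
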(event EventOnSetTargeting
  (params (bNewIsTargeting :type Boolean))
  (bind _output_get (Variables|Default|SetIsTargetting bNewIsTargeting))
  (Development|PrintString (select _output_get "B_Assault_Stark_Athena:: ADS" "B_Assault_Stark_Athena:: Hipfire"))
  (Audio|SpawnSoundAttached (select _output_get (Variables|Default|GetTransformIntoADSAudio) (Variables|Default|GetTransformOutOfADSAudio)) (Variables|Visual|GetWeaponMesh))
  (Effects|Components|ParticleSystem|SetVectorParameter (Variables|Effects|Muzzle|GetSpawned%20Muzzle%20System) "InitialColor+ADS" (Math|Conversions|ToVector(LinearColor) (select _output_get (Variables|Default|GetADSColor) (Variables|Default|GetHipfireColor))))
  (Utilities|Struct|SetmembersinFortNDCTracerInfo (Variables|Effects|GetNDCTracerFx) (Math|Conversions|ToInteger(Boolean) _output_get)))

(event FortWeapon|EventOnLocalUnEquip
  (|CallParentFunction)
  (if (Variables|Default|GetIsTargetting)
    (Audio|SpawnSoundAttached (Variables|Default|GetTransformOutOfADSAudio) (Variables|Visual|GetWeaponMesh))))

(fn ConstructionScript
  (|CallParentFunction))
