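; Weapon Blueprint for the Stark Athena assault rifle that switches its effects between hipfire and aim-down-sights.
; On targeting changes it stores the state, prints a debug string, plays transform-in or transform-out audio on the weapon mesh, and writes the state as an integer into the NDC tracer FX struct.
; Fire FX calls the parent, then attaches an ADS or hip muzzle Niagara system at the muzzle socket.
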(event EventOnPlayWeaponFireFX
  (params (bPersistentFire :type Boolean) (bSecondaryFire :type Boolean))
  (|CallParentFunction bPersistentFire bSecondaryFire)
  (Niagara|SpawnSystemAttached (select (Variables|Default|GetIsTargetting) "/HighTower/Items/Tomato/Tomato_Rifle/VFX/NS_Tomato_RifleMuzzle_ADS.NS_Tomato_RifleMuzzle_ADS" "/HighTower/Items/Tomato/Tomato_Rifle/VFX/NS_Tomato_RifleMuzzle_HIP.NS_Tomato_RifleMuzzle_HIP") (Variables|Visual|GetWeaponMesh) (Effects|GetMuzzleSocketName 0) "0, 0, 0" "0, 0, 0" "SnapToTargetIncludingScale" false true "ManualRelease"))

(event EventOnSetTargeting
  (params (bNewIsTargeting :type Boolean))
  (bind _output_get (Variables|Default|SetIsTargetting bNewIsTargeting))
  (Development|PrintString (select _output_get "B_Assault_Stark_Athena:: ADS" "B_Assault_Stark_Athena:: Hipfire"))
  (Audio|SpawnSoundAttached (select _output_get (Variables|Default|GetTransformIntoADSAudio) (Variables|Default|GetTransformOutOfADSAudio)) (Variables|Visual|GetWeaponMesh))
  (Utilities|Struct|SetmembersinFortNDCTracerInfo (Variables|Effects|GetNDCTracerFx) (Math|Conversions|ToInteger(Boolean) _output_get)))

(fn ConstructionScript
  (|CallParentFunction))
